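# Translucent UI material for the Plot Twist quest background, layering twinkling stars over a dark two-color gradient.
# The gradient uses rotated screen-aligned UVs with sine easing, alongside an eased radial-gradient mask.
# Front, mid and back star layers have scale, offset and opacity parameters, with back reusing the mid texture.
# Per-cell random noise minus looping triangle waves of differing durations drives each layer's tint and twinkle, and the layers add onto the gradient.
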Material /PlotTwistQuests/Items/QuestRewards/QuestRewards_PlotTwist/Materials/M_UI_PlotTwist_QuestBG
graph {
 properties {blendMode=BLEND_Translucent, materialDomain=MD_UI, pixelDepthOffsetMode=PDOM_Legacy, usedShadingModels=Default Lit}
 nodes [
  n1 = ComponentMask(None=MaterialFunctionCall(
    UVs=NamedRerouteUsage()
    "Rotation Angle (0-1)"=ScalarParameter() {defaultValue=0.75, group=Bkgd_Gradient, parameterName=Atmospherics_Mask_Angle}
   ) {materialFunction=&CustomRotator}.Rotated Values) {r=true}
  n2 = NamedRerouteUsage()
  n3 = NamedRerouteUsage()
  Add(
   A=LinearInterpolate(
    A=VectorParameter() {group=Bkgd_Gradient, parameterName=Bkgd_Gradient_Color1, defaultValue=<rgba 0.0008 0.0008 0.0052 1>}.RGB
    B=VectorParameter() {group=Bkgd_Gradient, parameterName=Bkgd_Gradient_Color2, defaultValue=<rgba 0.0058 0.005 0.0312 1>}.RGB
    Alpha=MaterialFunctionCall("Value (0-1)"=@n1) {materialFunction=&MF_UI_EaseCurves}.EaseInSine
   )
   B=Add(A=NamedRerouteUsage(), B=Add(A=@n2, B=@n3))
  )
  MaterialFunctionCall("Value (0-1)"=MaterialFunctionCall(GradientStart=ScalarParameter() {parameterName=MaskStart}, GradientEnd=ScalarParameter() {defaultValue=1, parameterName=MaskEnd}) {materialFunction=&MF_UI_RadialGradient}.Smooth) {materialFunction=&MF_UI_EaseCurves}
  NamedRerouteDeclaration(None=MaterialFunctionCall!(ScreenUV, TextureSize, ScalePivotPoint) {materialFunction=/FortUILibrary/MaterialFunctions/Basics/MF_UI_ScreenSpaceUVScaling}.Scaled to Fit) {name=Screen Alignment, nodeColor=<rgba 0 0 0 1>}
  n4 = Constant() {r=20}
  n5 = Floor(None=MaterialFunctionCall!(UV=NamedRerouteUsage(), Tiling X=@n4, Tiling Y=@n4) {materialFunction=/FortUILibrary/MaterialFunctions/Basics/MF_UI_Pattern}.Raw UVs)
  n6 = MaterialFunctionCall!(Animation Duration=Constant() {r=5}, Delay=Constant(), Reset to Default) {materialFunction=/FortUILibrary/MaterialFunctions/Time/MF_UI_AnimationDelay}
  n7 = Multiply(A=Min(A=@n6.Result, B=OneMinus(None=@n6.Result))) {constB=2}
  n8 = Saturate(None=Subtract(
    A=MaterialFunctionCall!(UV=@n5, Seed) {desc=Noise, materialFunction=&MF_UI_Random}.Result
    B=Multiply(A=@n7) {constB=0.5, desc=Maximum Falloff Amount}
   ))
  n9 = MaterialFunctionCall(
   A=Constant3Vector() {constant=<rgba 1 1 1 0>}
   B=Constant3Vector() {constant=<rgba 0.526 0.7151 1 0>}
   C=Constant3Vector() {constant=<rgba 1 0.8867 0.8385 1>}
   Alpha=@n8
  ) {materialFunction=&3ColorBlend}
  n10 = Multiply(A=NamedRerouteUsage(), B=ScalarParameter() {defaultValue=1, group=Star_Front, parameterName=Front_Star_Scale})
  n11 = AppendVector(
   A=ScalarParameter() {group=Star_Front, parameterName=Front_Star_Offset_X}
   B=ScalarParameter() {group=Star_Front, parameterName=Front_Star_Offset_Y}
  )
  n12 = NamedRerouteUsage()
  n13 = Multiply(
   A=TextureSample(UVs=Add(A=@n10, B=@n11)) {texture=/PlotTwistQuests/Items/QuestRewards/QuestRewards_PlotTwist/Textures/T_UI_PlotTwist_Stars_Front}.RGB
   B=Saturate(None=Add(A=@n12) {constB=0.4})
  )
  n14 = ScalarParameter() {defaultValue=1, group=Star_Front, parameterName=Front_Star_Opacity, sliderMax=1}
  NamedRerouteDeclaration(None=Multiply(A=Multiply(A=@n9.Result) {desc=Maximum Falloff Amount}, B=Multiply(A=@n13, B=@n14))) {name=Front Stars, nodeColor=<rgba 0 0 0 1>}
  n15 = Constant() {r=20}
  n16 = Floor(None=MaterialFunctionCall(UV=NamedRerouteUsage(), Tiling X=@n15, Tiling Y=@n15) {materialFunction=&MF_UI_Pattern}.Raw UVs)
  n17 = Constant() {r=0.35}
  n18 = MaterialFunctionCall(Animation Duration=Constant() {r=6.5}, Delay=Constant()) {materialFunction=&MF_UI_AnimationDelay}
  n19 = Multiply(A=Min(A=@n18.Result, B=OneMinus(None=@n18.Result))) {constB=2}
  n20 = Saturate(None=Subtract(
    A=MaterialFunctionCall(UV=@n16, Seed=@n17) {desc=Noise, materialFunction=&MF_UI_Random}.Result
    B=Multiply(A=@n19) {constB=0.5, desc=Maximum Falloff Amount}
   ))
  n21 = MaterialFunctionCall(
   A=Constant3Vector() {constant=<rgba 1 0.7851 0.6719 1>}
   B=Constant3Vector() {constant=<rgba 1 1 1 1>}
   C=Constant3Vector() {constant=<rgba 0.7448 0.77 1 1>}
   Alpha=@n20
  ) {materialFunction=&3ColorBlend}
  n22 = Multiply(A=NamedRerouteUsage(), B=ScalarParameter() {defaultValue=1, group=Star_Mid, parameterName=Mid_Star_Scale})
  n23 = AppendVector(A=ScalarParameter() {group=Star_Mid, parameterName=Mid_Star_Offset_X}, B=ScalarParameter() {group=Star_Mid, parameterName=Mid_Star_Offset_Y})
  n24 = Multiply(
   A=TextureSample(UVs=Add(A=@n22, B=@n23)) {texture=/PlotTwistQuests/Items/QuestRewards/QuestRewards_PlotTwist/Textures/T_UI_PlotTwist_Stars_Mid}.RGB
   B=NamedRerouteUsage()
  )
  n25 = ScalarParameter() {defaultValue=1, group=Star_Mid, parameterName=Mid_Star_Opacity, sliderMax=1}
  NamedRerouteDeclaration(None=Multiply(A=Multiply(A=@n21.Result) {desc=Maximum Falloff Amount}, B=Multiply(A=@n24, B=@n25))) {name=Mid Stars, nodeColor=<rgba 0 0 0 1>}
  n26 = Constant() {r=20}
  n27 = Floor(None=MaterialFunctionCall(UV=NamedRerouteUsage(), Tiling X=@n26, Tiling Y=@n26) {materialFunction=&MF_UI_Pattern}.Raw UVs)
  n28 = Constant() {r=0.5}
  n29 = MaterialFunctionCall(Animation Duration=Constant() {r=10}, Delay=Constant()) {materialFunction=&MF_UI_AnimationDelay}
  n30 = Multiply(A=Min(A=@n29.Result, B=OneMinus(None=@n29.Result))) {constB=2}
  n31 = Saturate(None=Subtract(
    A=MaterialFunctionCall(UV=@n27, Seed=@n28) {desc=Noise, materialFunction=&MF_UI_Random}.Result
    B=Multiply(A=@n30) {constB=0.5, desc=Maximum Falloff Amount}
   ))
  n32 = MaterialFunctionCall(
   A=Constant3Vector() {constant=<rgba 0.8882 0.7257 0.9479 1>}
   B=Constant3Vector() {constant=<rgba 0.5341 0.693 0.9323 0>}
   C=Constant3Vector() {constant=<rgba 1 1 1 1>}
   Alpha=@n31
  ) {materialFunction=&3ColorBlend}
  n33 = Multiply(A=NamedRerouteUsage(), B=ScalarParameter() {defaultValue=4, group=Star_Bkgd, parameterName=Back_Star_Scale})
  n34 = AppendVector(
   A=ScalarParameter() {group=Star_Bkgd, parameterName=Back_Star_Offset_X}
   B=ScalarParameter() {group=Star_Bkgd, parameterName=Back_Star_Offset_Y}
  )
  n35 = Multiply(
   A=TextureSample(UVs=Add(A=@n33, B=@n34)) {texture=&T_UI_PlotTwist_Stars_Mid}.RGB
   B=NamedRerouteUsage()
  )
  n36 = ScalarParameter() {defaultValue=1, group=Star_Bkgd, parameterName=Back_Star_Opacity, sliderMax=1}
  NamedRerouteDeclaration(None=Multiply(A=Multiply(A=@n32.Result) {desc=Maximum Falloff Amount}, B=Multiply(A=@n35, B=@n36))) {name=Back Stars, nodeColor=<rgba 0 0 0 1>}
  NamedRerouteDeclaration(None=@n8) {name=NoiseFront, nodeColor=<rgba 0 0 0 1>}
  NamedRerouteDeclaration(None=@n20) {name=NoiseMid, nodeColor=<rgba 0 0 0 1>}
  NamedRerouteDeclaration(None=@n31) {name=NoiseBack, nodeColor=<rgba 0 0 0 1>}
 ]
}
