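# Quest rewards panel widget for PlotTwist that shows decorative reward art: a character, a Sith Infiltrator glider and an emote over a quest background.
# A fixed-width SizeBox holds an Overlay of Images positioned by render transforms, with both background and foreground glow layers collapsed.
# On Construct it loops FloatAnim endlessly, bobbing the glider and emote vertically in opposite-sized offsets around their resting positions.
# The Default animation is empty and the Image_LoadingScreen variable is unused.
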WidgetBlueprint /PlotTwistQuests/Items/QuestRewards/QuestRewards_PlotTwist/WBP_QuestRewards_Panel_PlotTwist
animations [
 Default {
  binding_count=0
  length_seconds=5
  object=&self:Default
  sequence {bindings []}
 }
 FloatAnim {
  binding_count=2
  length_seconds=2.5
  object=&self:FloatAnim
  sequence {bindings [
    Image_Glider {
     id=#1
     possessed_class=Image
     tracks [{
       type=MovieScene2DTransformTrack
       sections [{
         type=MovieScene2DTransformSection
         end_frame=50
         start_frame=0
         channels [
          Translation.X {type=MovieSceneScriptingFloatChannel, const [-197,1]}
          Translation.Y {type=MovieSceneScriptingFloatChannel, keys [[0,-177],[25,-185],[50,-177]]}
          Angle {type=MovieSceneScriptingFloatChannel, const [10,1]}
          Scale.X {type=MovieSceneScriptingFloatChannel, const [-0.9,1]}
          Scale.Y {type=MovieSceneScriptingFloatChannel, const [0.9,1]}
          Shear.X {type=MovieSceneScriptingFloatChannel, const [0,1]}
          Shear.Y {type=MovieSceneScriptingFloatChannel, const [0,1]}
         ]
        }]
      }]
    }
    Image_Emote {
     id=#2
     possessed_class=Image
     tracks [{
       type=MovieScene2DTransformTrack
       sections [{
         type=MovieScene2DTransformSection
         end_frame=50
         start_frame=0
         channels [
          Translation.X {type=MovieSceneScriptingFloatChannel, const [-203,1]}
          Translation.Y {type=MovieSceneScriptingFloatChannel, keys [[0,228],[25,214.0029],[50,228]]}
          Angle {type=MovieSceneScriptingFloatChannel, const [0,1]}
          Scale.X {type=MovieSceneScriptingFloatChannel, const [0.6,1]}
          Scale.Y {type=MovieSceneScriptingFloatChannel, const [0.6,1]}
          Shear.X {type=MovieSceneScriptingFloatChannel, const [0,1]}
          Shear.Y {type=MovieSceneScriptingFloatChannel, const [0,1]}
         ]
        }]
      }]
    }
   ]}
 }
]
graphs [EventGraph {nodes [
   Event Construct()
   > PlayAnimation(InAnimation=Get FloatAnim() -> (FloatAnim:Widget Animation Object Reference), NumLoopsToPlay=0)
  ]}]
variables [Image_LoadingScreen {type=Image, default=null}]
widgets {
 info {inheritedWidgetCount=0, parentClass=&UserWidget, rootWidgetClass=&SizeBox, widgetCount=8}
 widgets [
  SizeBox_0 {widgetClassPath=&SizeBox, widget=SizeBox_0, properties {bOverride_WidthOverride=true, heightOverride=940, widthOverride=847}}
  Overlay_0 {widgetClassPath=&Overlay, parent=SizeBox_0, widget=Overlay_0, slot {class=SizeBoxSlot}}
  Image_Background {
   widgetClassPath=&Image
   bIsVariable=true
   parent=Overlay_0
   widget=Image_Background
   properties {
    brush {
     bIsDynamicallyLoaded=false
     drawAs=Image
     imageType=NoImage
     mirroring=NoMirror
     resourceName=None
     resourceObject=/PlotTwistQuests/Items/QuestRewards/QuestRewards_PlotTwist/Materials/MI_UI_PlotTwist_QuestBG_Inst
     tiling=NoTile
     imageSize=<512 512>
     margin=<ltrb 0 0 0 0>
     outlineSettings {
      bUseBrushTransparency=false
      roundingType=HalfHeightRadius
      width=0
      color {colorUseRule=UseColor_Specified, specifiedColor=<rgba 0 0 0 0>}
      cornerRadii=<0 0 0 0>
     }
     tintColor {colorUseRule=UseColor_Specified, specifiedColor=<rgba 1 1 1 1>}
     uVRegion {bIsValid=false, max=<0 0>, min=<0 0>}
    }
    renderTransform {angle=0, scale=<2.5 2.5>, shear=<0 0>, translation=<-169 0>}
    renderTransformPivot=<0.4366 0.5>
   }
   slot {class=OverlaySlot, properties {horizontalAlignment=HAlign_Center, verticalAlignment=VAlign_Center}}
  }
  Image_BackgroundGlow {
   widgetClassPath=&Image
   bIsVariable=true
   parent=Overlay_0
   widget=Image_BackgroundGlow
   properties {
    visibility=Collapsed
    brush {
     bIsDynamicallyLoaded=false
     drawAs=Image
     imageType=NoImage
     mirroring=NoMirror
     resourceName=None
     resourceObject=/PlotTwistQuests/Items/QuestRewards/QuestRewards_PlotTwist/Materials/MI_UI_QuestRewards_Glow_PlotTwist
     tiling=NoTile
     imageSize=<1025 1025>
     margin=<ltrb 0 0 0 0>
     outlineSettings {
      bUseBrushTransparency=false
      roundingType=HalfHeightRadius
      width=0
      color {colorUseRule=UseColor_Specified, specifiedColor=<rgba 0 0 0 0>}
      cornerRadii=<0 0 0 0>
     }
     tintColor {colorUseRule=UseColor_Specified, specifiedColor=<rgba 1 1 1 1>}
     uVRegion {bIsValid=false, max=<0 0>, min=<0 0>}
    }
    colorAndOpacity=<rgba 1 1 1 0.6714>
    renderTransform {angle=0, scale=<1.3 1.3>, shear=<0 0>, translation=<0 -100>}
   }
   slot {class=OverlaySlot, properties {horizontalAlignment=HAlign_Center, verticalAlignment=VAlign_Center}}
  }
  Image_Character {
   widgetClassPath=&Image
   bIsVariable=true
   parent=Overlay_0
   widget=Image_Character
   properties {
    brush {
     bIsDynamicallyLoaded=false
     drawAs=Image
     imageType=NoImage
     mirroring=NoMirror
     resourceName=None
     resourceObject=/PlotTwistQuests/Items/QuestRewards/QuestRewards_PlotTwist/Materials/M_UI_PlotTwist
     tiling=NoTile
     imageSize=<500 500>
     margin=<ltrb 0 0 0 0>
     outlineSettings {
      bUseBrushTransparency=false
      roundingType=HalfHeightRadius
      width=0
      color {colorUseRule=UseColor_Specified, specifiedColor=<rgba 0 0 0 0>}
      cornerRadii=<0 0 0 0>
     }
     tintColor {colorUseRule=UseColor_Specified, specifiedColor=<rgba 1 1 1 1>}
     uVRegion {bIsValid=false, max=<0 0>, min=<0 0>}
    }
    renderTransform {angle=3, scale=<1.9 1.9>, shear=<0 0>, translation=<222.1261 0>}
    renderTransformPivot=<0.4 0.5>
   }
   slot {class=OverlaySlot, properties {horizontalAlignment=HAlign_Center, verticalAlignment=VAlign_Center}}
  }
  Image_Glider {
   widgetClassPath=&Image
   bIsVariable=true
   parent=Overlay_0
   widget=Image_Glider
   properties {
    brush {
     bIsDynamicallyLoaded=false
     drawAs=Image
     imageType=NoImage
     mirroring=NoMirror
     resourceName=None
     resourceObject=/PlotTwistQuests/UI/Textures/T_UI_PlotTwist_SithInfiltrator1
     tiling=NoTile
     imageSize=<512 512>
     margin=<ltrb 0 0 0 0>
     outlineSettings {
      bUseBrushTransparency=false
      roundingType=HalfHeightRadius
      width=0
      color {colorUseRule=UseColor_Specified, specifiedColor=<rgba 0 0 0 0>}
      cornerRadii=<0 0 0 0>
     }
     tintColor {colorUseRule=UseColor_Specified, specifiedColor=<rgba 1 1 1 1>}
     uVRegion {bIsValid=false, max=<0 0>, min=<0 0>}
    }
    renderTransform {angle=10, scale=<-0.9 0.9>, shear=<0 0>, translation=<-197 -177>}
   }
   slot {class=OverlaySlot, properties {horizontalAlignment=HAlign_Center, verticalAlignment=VAlign_Center}}
  }
  Image_Emote {
   widgetClassPath=&Image
   bIsVariable=true
   parent=Overlay_0
   widget=Image_Emote
   properties {
    brush {
     bIsDynamicallyLoaded=false
     drawAs=Image
     imageType=NoImage
     mirroring=NoMirror
     resourceName=None
     resourceObject=/PlotTwistQuests/Items/QuestRewards/QuestRewards_PlotTwist/Materials/M_UI_PlotTwist_Emote
     tiling=NoTile
     imageSize=<512 512>
     margin=<ltrb 0 0 0 0>
     outlineSettings {
      bUseBrushTransparency=false
      roundingType=HalfHeightRadius
      width=0
      color {colorUseRule=UseColor_Specified, specifiedColor=<rgba 0 0 0 0>}
      cornerRadii=<0 0 0 0>
     }
     tintColor {colorUseRule=UseColor_Specified, specifiedColor=<rgba 1 1 1 1>}
     uVRegion {bIsValid=false, max=<0 0>, min=<0 0>}
    }
    renderTransform {angle=0, scale=<0.6 0.6>, shear=<0 0>, translation=<-203 228>}
   }
   slot {class=OverlaySlot, properties {horizontalAlignment=HAlign_Center, verticalAlignment=VAlign_Center}}
  }
  Image_ForegroundGlow {
   widgetClassPath=&Image
   bIsVariable=true
   parent=Overlay_0
   widget=Image_ForegroundGlow
   properties {
    visibility=Collapsed
    brush {
     bIsDynamicallyLoaded=false
     drawAs=Image
     imageType=NoImage
     mirroring=NoMirror
     resourceName=None
     resourceObject=/PlotTwistQuests/Items/QuestRewards/QuestRewards_PlotTwist/Materials/MI_UI_QuestRewards_ForegroundGlow_PlotTwist
     tiling=NoTile
     imageSize=<1024 1024>
     margin=<ltrb 0 0 0 0>
     outlineSettings {
      bUseBrushTransparency=false
      roundingType=HalfHeightRadius
      width=0
      color {colorUseRule=UseColor_Specified, specifiedColor=<rgba 0 0 0 0>}
      cornerRadii=<0 0 0 0>
     }
     tintColor {colorUseRule=UseColor_Specified, specifiedColor=<rgba 1 1 1 1>}
     uVRegion {bIsValid=false, max=<0 0>, min=<0 0>}
    }
    colorAndOpacity=<rgba 1 1 1 0.0417>
    renderTransform {angle=0, scale=<1.25 1.25>, shear=<0 0>, translation=<0 -83>}
   }
   slot {class=OverlaySlot, properties {horizontalAlignment=HAlign_Right, verticalAlignment=VAlign_Center, padding=<ltrb 0 0 -50 -200>}}
  }
 ]
}
